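# Audio level sequence for the tdm0200 segment of the GrillMoon Countdown cinematic, driving its sound and POI audio cues.
# A master audio track named tdm holds a single section.
# Event trigger tracks on the music and ambience components, both children of the Day 2 sequence runner, fire events whose target functions are not shown.
# The mask timer Blueprint is also bound but has no tracks listed.
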LevelSequence /GrillMoon_Countdown/Sequence/tdm/tdm0200/GrillMoon_Countdown_tdm0200_aud
bindings [
 BP_GrillMoon_Day2_SequenceRunner {id=#1, possessed_class=BP_GrillMoon_Day2_SequenceRunner_C, tracks []}
 BP_GrillMoon_POI_MusicComponent {
  id=#2
  parent_id=#1
  possessed_class=BP_GrillMoon_POI_MusicComponent_C
  tracks [{
    type=MovieSceneEventTrack
    sections [{
      type=MovieSceneEventTriggerSection
      channels [None {
        type=MovieSceneScriptingEventChannel
        keys [[3624, "<Struct 'MovieSceneEvent' (0x850243114a0) {}>"], [4060.96, "<Struct 'MovieSceneEvent' (0x8502431d2a0) {}>"]]
       }]
     }]
   }]
 }
 GrillMoon_Countdown_BPmaskTimer {id=#3, possessed_class=GrillMoon_Countdown_BPmaskTimer_C, tracks []}
 BP_GrillMoon_POI_AmbienceComponent {
  id=#4
  parent_id=#1
  possessed_class=BP_GrillMoon_POI_AmbienceComponent_C
  tracks [{
    type=MovieSceneEventTrack
    sections [{
      type=MovieSceneEventTriggerSection
      channels [None {type=MovieSceneScriptingEventChannel, const ["<Struct 'MovieSceneEvent' (0x85024312300) {}>", 1]}]
     }]
   }]
 }
]
display_rate {denominator=1, numerator=60}
tick_resolution {denominator=1, numerator=24000}
tracks [{
  type=MovieSceneAudioTrack
  display_name=tdm
  sections [{
    type=MovieSceneAudioSection
    end_frame=4201
    start_frame=3649
    channels [
     Volume {type=MovieSceneScriptingFloatChannel, default=1}
     Pitch {type=MovieSceneScriptingFloatChannel, default=1}
     AttachActor {type=MovieSceneScriptingActorReferenceChannel}
    ]
   }]
 }]
